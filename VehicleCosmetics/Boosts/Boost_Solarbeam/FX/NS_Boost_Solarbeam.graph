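# Niagara system for the Solarbeam vehicle boost cosmetic, emitting a trail, particles and glow behind the car from a single boost position plus car offset.
# A two-renderer ribbon streams backward, faster in frontend preview, with age- and time-driven sine waves animating its position and facing.
# Small sprites spray in a noisy, dragged cone, while large layered sprites form the nozzle glow.
# Each emitter applies painted-variant color overrides, with secondary palettes for ribbon and glow.
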NiagaraSystem /VehicleCosmetics/Boosts/Boost_Solarbeam/FX/NS_Boost_Solarbeam
dependencies {
 usedDataInterfaces [&NiagaraDataInterfaceCurve, &NiagaraDataInterfaceSkeletalMesh]
 usedDynamicInputs []
 usedModules [
  &self:SystemSpawnScript.SystemScriptSource.SystemScriptGraph.NiagaraNodeAssignment_0.SetVariables_#1
  &SystemState
 ]
}
emitters {
 NE_VC_SingleBoost_Ribbon_Parent {
  emitterUpdateScript [
   &EmitterState()
   &VC_FilteredSocketCount()
   &VC_RateQualityHandler(Epic Spawn Rate=40, High Spawn Rate=40, Low Spawn Rate=40, Medium Spawn Rate=40, Use Custom Rate?=true)
   &VC_SpawnRibbon_v1(SpawnRate=$Emitter.Spawn Rate)
  ]
  particleSpawnScript [
   &VC_PaintedVariantsOverride(
    Black=<rgba 0.0152 0.0152 0.016 2>
    BurntSienna=<rgba 0.25 0.0553 0.0077 1>
    Cobalt=<rgba 0.066 0.0838 0.6 1>
    Crimson=<rgba 0.8632 0.0173 0.0201 1>
    DefaultColor=<rgba 1 1 0.45 1>
    ForestGreen=<rgba 0.0533 0.9 0.0469 1>
    Grey=<rgba 0.12 0.12 0.12 0.9>
    Lime=<rgba 0.3024 0.9 0.0254 1>
    Orange=<rgba 0.52 0.1596 0.0052 1.2>
    Pink=<rgba 0.8 0.24 0.8 0.85>
    Purple=<rgba 0.7084 0.0513 0.9911 0.4>
    R1 Black:LinearColor=<rgba 0.008 0.008 0.008 2>
    R1 BurntSienna:LinearColor=<rgba 0.09 0.0232 0.0031 1.2>
    R1 Cobalt:LinearColor=<rgba 0.0167 0 0.5 1.2>
    R1 Crimson:LinearColor=<rgba 0.2918 0.011 0.0097 1>
    R1 DefaultColor=<rgba 0.6 0.4777 0.15 1>
    R1 ForestGreen:LinearColor=<rgba 0.0086 0.2 0.0078 1>
    R1 Grey:LinearColor=<rgba 0.0327 0.0327 0.035 1.5>
    R1 Lime:LinearColor=<rgba 0.1163 0.25 0.0069 1.2>
    R1 Orange:LinearColor=<rgba 0.38 0.0606 0.0054 1.2>
    R1 Pink:LinearColor=<rgba 0.9 0.135 0.7129 0.8>
    R1 Purple:LinearColor=<rgba 0.1375 0 0.55 1>
    R1 Saffron:LinearColor=<rgba 0.2 0.2 0.008 1>
    R1 SkyBlue:LinearColor=<rgba 0 0.16 0.9 1>
    R1 TitaniumWhite:LinearColor=<rgba 0.2731 0.2731 0.28 1>
    Saffron=<rgba 0.7 0.7 0.007 1>
    SkyBlue=<rgba 0.01 0.7362 1 1>
    TitaniumWhite=<rgba 0.76 0.76 0.76 1>
    Use R1 Painted Overrides=true
   )
   &VC_InitializeRibbon_v1(
    Color=$Particles.Color
    Custom Offset=true
    Lifetime=0.9
    Local Vector Offset=&AddVectorToPosition(
     Position=&AddVectorToPosition(
      Position=<-80 0 0>
      Vector=&AddVectorToPosition(Position=$User.DTCarOffset, Vector=Linked() {linkedVariable {name=User.SingleBoostPos, type=NiagaraPosition}})
     )
     Vector=<0 0 0>
    )
    Offset Amount=1
    Ribbon Width=50
   )
   &AddVelocity(
    Velocity=<-1500 0 0>
    Velocity Speed Scale=&MakeCustomFloatFromBool(Bool=$User.bIsFrontendPreview, True Float=1.3, False Float=0.8)
   ) {module=AddVelocity001}
  ]
  particleUpdateScript [
   &ParticleState()
   &SolveForcesAndVelocity(Clamp Velocity=false, Limit Acceleration=false)
   &VC_UpdateRibbon_v1()
   &ScaleColor(
    Scale Alpha=&FloatFromCurve(FloatCurve=DataInterface() {propertyValues {Curve {
        defaultValue=1
        keys [
         {
          arriveTangent=2.1825
          arriveTangentWeight=0.1243
          interpMode=RCIM_Cubic
          leaveTangent=2.1825
          leaveTangentWeight=0.1243
          tangentMode=RCTM_User
          tangentWeightMode=RCTWM_WeightedBoth
          time=0
          value=-0.1
         }
         {arriveTangent=-0.013, interpMode=RCIM_Cubic, leaveTangent=-0.013, tangentMode=RCTM_User, time=0.1617, value=0.9847}
         {
          arriveTangent=0.0238
          arriveTangentWeight=0.4421
          interpMode=RCIM_Cubic
          leaveTangent=0.0238
          leaveTangentWeight=0.4421
          tangentMode=RCTM_User
          tangentWeightMode=RCTWM_WeightedBoth
          time=0.79
          value=0
         }
        ]
       }}})
    ScaleA=true
    ScaleRGB=true
   )
   &ScaleRibbonWidth(Ribbon Width Scale=&FloatFromCurve(FloatCurve=DataInterface() {propertyValues {Curve {
        defaultValue=1
        keys [
         {arriveTangent=1.071, interpMode=RCIM_Cubic, leaveTangent=1.071, tangentMode=RCTM_User, time=0.1, value=0.35}
         {arriveTangent=-0.0504, interpMode=RCIM_Cubic, leaveTangent=-0.0504, tangentMode=RCTM_User, time=0.6, value=0.73}
         {arriveTangent=-1.632, interpMode=RCIM_Cubic, leaveTangent=-1.632, tangentMode=RCTM_User, time=0.8, value=0.5625}
        ]
       }}}))
   /SetVariables(
    Particles.ScaleColorOutterRibbon:LinearColor=&MakeLinearColorFromVectorAndFloat(
     "Vector (RGB)"=&MakeVectorFromLinearColorRGB(Color=$Particles.R1 Color)
     "Float (Alpha)"=&MakeCustomFloatFromBool(
      Bool=&SetBoolByFloatComparison(A=&MakeFloatFromInt(Int=$System.ColorVariantID), B=2, Comparison Type=::NewEnumerator4)
      True Float=&FloatFromCurve(
       FloatCurve=DataInterface() {propertyValues {Curve {
          defaultValue=1
          keys [
           {
            arriveTangent=5.264
            arriveTangentWeight=0.3718
            interpMode=RCIM_Cubic
            leaveTangent=5.264
            leaveTangentWeight=0.3718
            tangentMode=RCTM_User
            tangentWeightMode=RCTWM_WeightedBoth
            time=0
            value=0
           }
           {
            arriveTangent=-0.0402
            arriveTangentWeight=0.0961
            interpMode=RCIM_Cubic
            leaveTangent=-0.0402
            leaveTangentWeight=0.2784
            tangentMode=RCTM_Break
            tangentWeightMode=RCTWM_WeightedBoth
            time=0.2514
            value=0.8863
           }
           {
            arriveTangent=0.0494
            arriveTangentWeight=0.1466
            interpMode=RCIM_Cubic
            leaveTangent=0.0494
            leaveTangentWeight=0.1466
            tangentMode=RCTM_User
            tangentWeightMode=RCTWM_WeightedBoth
            time=0.8
            value=0
           }
          ]
         }}}
       Scale Curve=&MakeFloatFromLinearColor(LinearColor=$Particles.R1 Color, Channel:ENiagaraFloatFromLinearColorOptions=::NewEnumerator3)
      )
      False Float=&FloatFromCurve(
       FloatCurve=DataInterface() {propertyValues {Curve {
          defaultValue=1
          keys [
           {
            arriveTangent=0.0174
            arriveTangentWeight=0.1004
            interpMode=RCIM_Cubic
            leaveTangent=0.0174
            leaveTangentWeight=0.1004
            tangentMode=RCTM_User
            tangentWeightMode=RCTWM_WeightedBoth
            time=0
            value=0
           }
           {
            arriveTangent=-0.0402
            arriveTangentWeight=0.0961
            interpMode=RCIM_Cubic
            leaveTangent=-0.0402
            leaveTangentWeight=0.1892
            tangentMode=RCTM_Break
            tangentWeightMode=RCTWM_WeightedBoth
            time=0.2514
            value=0.8863
           }
           {
            arriveTangent=-0.0644
            arriveTangentWeight=0.2235
            interpMode=RCIM_Cubic
            leaveTangent=-0.0644
            leaveTangentWeight=0.2235
            tangentMode=RCTM_User
            tangentWeightMode=RCTWM_WeightedBoth
            time=0.8
            value=0
           }
          ]
         }}}
       Scale Curve=&MakeFloatFromLinearColor(LinearColor=$Particles.R1 Color, Channel:ENiagaraFloatFromLinearColorOptions=::NewEnumerator3)
      )
     )
    )
    Particles.ScaleRibbonWidthOutter:NiagaraFloat=&FloatFromCurve(
     FloatCurve=DataInterface() {propertyValues {Curve {
        defaultValue=1
        keys [
         {arriveTangent=1.071, interpMode=RCIM_Cubic, leaveTangent=1.071, tangentMode=RCTM_User, time=0.1, value=0.35}
         {arriveTangent=-0.0504, interpMode=RCIM_Cubic, leaveTangent=-0.0504, tangentMode=RCTM_User, time=0.6, value=0.62}
         {arriveTangent=-1.632, interpMode=RCIM_Cubic, leaveTangent=-1.632, tangentMode=RCTM_User, time=0.8, value=0.4605}
        ]
       }}}
     Scale Curve=150
    )
   ) {isSetParameters=true}
   /SetVariables(Particles.Position=&AddVectorToPosition(
     Position=$Particles.Position
     Vector=/VehicleCosmetics/FX/Niagara/Modules/VC_ConvertVectorToWorld(Vector:Vector3f=&MakeVector(
       X=0
       Y=0
       Z=&FloatFromCurve(
        FloatCurve=DataInterface() {propertyValues {Curve {keys [
            <at 0 0 RCIM_Cubic>
            {arriveTangent=80, interpMode=RCIM_Cubic, leaveTangent=80, time=0.2, value=40}
            {arriveTangent=3.1163, interpMode=RCIM_Cubic, leaveTangent=3.1163, tangentMode=RCTM_User, time=0.5, value=40}
            {arriveTangent=-14.8689, interpMode=RCIM_Cubic, leaveTangent=-14.8689, tangentMode=RCTM_User, time=0.8, value=0}
           ]}}}
        Scale Curve=0.08
       )
      ))
    )) {isSetParameters=true}
   /SetVariables(Particles.Position=&AddVectorToPosition(
     Position=$Particles.Position
     Vector=&TransformVector(
      Vector=&MakeVector(
       X=0
       Y=&Sine(
        Normalized Angle=&Add_Float(
         A=&Multiply_Float(A=$Particles.NormalizedAge, B=-1.2)
         B=&Multiply_Float(A=$Emitter.LoopedAge, B=2.1)
        )
        Period=1
        Scale=&FloatFromCurve(
         FloatCurve=DataInterface() {propertyValues {Curve {defaultValue=1, keys [{tangentMode=RCTM_User, time=0, value=0.7}, {tangentMode=RCTM_User, time=0.53, value=1.1}]}}}
         Scale Curve=3
        )
        Bias=0
       )
       Z=&Sine(
        Normalized Angle=&Add_Float(
         A=&Multiply_Float(A=$Particles.NormalizedAge, B=-1.3)
         B=&Multiply_Float(A=$Emitter.LoopedAge, B=2.1)
        )
        Period=1
        Scale=&FloatFromCurve(
         FloatCurve=DataInterface() {propertyValues {Curve {defaultValue=1, keys [{tangentMode=RCTM_User, time=0, value=0.7}, {tangentMode=RCTM_User, time=0.53, value=1.1}]}}}
         Scale Curve=3
        )
        Bias=0
       )
      )
      Destination Space=::Simulation
     )
    )) {isSetParameters=true}
   /SetVariables(Particles.RibbonFacing=&MaskVectorBySpawnGroup(
     Masked Vector=&MakeVector(
      X=1
      Y=0
      Z=&Sine(
       Normalized Angle=&Add_Float(
        A=&Multiply_Float(A=$Particles.NormalizedAge, B=1)
        B=&Multiply_Float(A=$Emitter.LoopedAge, B=1)
       )
       Period=1
       Scale=&FloatFromCurve(FloatCurve=DataInterface() {propertyValues {Curve {
           defaultValue=1
           keys [{arriveTangent=4.6832, interpMode=RCIM_Cubic, leaveTangent=4.6832, tangentMode=RCTM_User, time=0, value=0}, <at 0.5 1 RCIM_Cubic>]
          }}})
       Bias=0
      )
     )
     Passthrough Vector=&MakeVector(
      X=-1
      Y=0
      Z=&Sine(
       Normalized Angle=&Add_Float(
        A=&Multiply_Float(A=$Particles.NormalizedAge, B=1)
        B=&Multiply_Float(A=$Emitter.LoopedAge, B=1)
       )
       Period=1
       Scale=&FloatFromCurve(FloatCurve=DataInterface() {propertyValues {Curve {
           defaultValue=1
           keys [{arriveTangent=4.6832, interpMode=RCIM_Cubic, leaveTangent=4.6832, tangentMode=RCTM_User, time=0, value=0}, <at 0.5 1 RCIM_Cubic>]
          }}})
       Bias=0
      )
     )
    )) {isSetParameters=true}
  ]
  rendererClasses [&NiagaraRibbonRendererProperties]
  renderers [
   {rendererClass=&NiagaraRibbonRendererProperties}
   {rendererClass=&NiagaraRibbonRendererProperties, rendererIndex=1}
  ]
 }
 NE_VC_SingleBoost_Sprite_Parent {
  emitterUpdateScript [
   &EmitterState()
   &SpawnRate(Spawn Probability=0.7, SpawnRate=11, Use Spawn Probability=true)
  ]
  particleSpawnScript [
   &VC_PaintedVariantsOverride(
    Black=<rgba 0.035 0.035 0.035 1>
    BurntSienna=<rgba 0.4 0.1973 0.02 1>
    Cobalt=<rgba 0.25 0.5811 1 1>
    Crimson=<rgba 0.45 0.1575 0.1654 1>
    DefaultColor=<rgba 0.9781 1 0.1 1>
    ForestGreen=<rgba 0 0.7 0.0463 1>
    Grey=<rgba 0.4 0.4 0.4 1>
    Lime=<rgba 0.4083 0.7 0 1>
    Orange=<rgba 1 0.8395 0.0369 1>
    Pink=<rgba 1 0.5 0.84 1>
    Purple=<rgba 1 0.09 1 1>
    Saffron=<rgba 1 1 0.1 1>
    SkyBlue=<rgba 0.5 1 1 1>
   )
   &InitializeParticle(
    Color=$Particles.Color
    Lifetime=0.85
    Position Offset=&AddVectorToPosition(
     Position=<-100 0 0>
     Vector=&AddVectorToPosition(Position=$User.DTCarOffset, Vector=Linked() {linkedVariable {name=User.SingleBoostPos, type=NiagaraPosition}})
    )
    Position Offset Coordinate Space=::Local
    Sprite Rotation Mode=::NewEnumerator2
    Sprite Size Mode=::NewEnumerator3
    Uniform Sprite Size=13
    UsePositionOffset=true
    Write Material Random=true
   )
   &ShapeLocation(MaskSpawnGroup=false, Override Seed=false, Sphere Radius=10, UseOwnerScale=true)
   &AddVelocity(
    Cone Angle=16
    Module.UseConeFalloff=false
    Module.UseVelDistribution=false
    Override Seed=false
    Velocity Mode=::NewEnumerator2
    Velocity Speed=-700
   )
  ]
  particleUpdateScript [
   &ParticleState()
   &ScaleColor(
    Scale Alpha=&FloatFromCurve(FloatCurve=DataInterface() {propertyValues {Curve {
        defaultValue=1
        keys [
         {arriveTangent=-0.0824, interpMode=RCIM_Cubic, leaveTangent=-0.0824, leaveTangentWeight=0.0836, tangentMode=RCTM_User, time=0.65, value=1}
         {arriveTangent=-9.0725, arriveTangentWeight=0.7606, interpMode=RCIM_Cubic, leaveTangent=-9.0725, tangentMode=RCTM_User, time=0.95, value=0}
        ]
       }}})
    ScaleA=true
    ScaleRGB=true
   )
   &ScaleSpriteSize(Uniform Curve Sprite Scale=DataInterface() {propertyValues {Curve {
       defaultValue=1
       keys [
        {arriveTangent=-0.0824, interpMode=RCIM_Cubic, leaveTangent=-0.0824, leaveTangentWeight=0.0836, tangentMode=RCTM_User, time=0.75, value=1}
        {arriveTangent=-9.0725, arriveTangentWeight=0.7606, interpMode=RCIM_Cubic, leaveTangent=-9.0725, tangentMode=RCTM_User, time=0.95, value=0}
       ]
      }}})
   &SpriteRotationRate(Rotation Rate=100)
   &CurlNoiseForce(Bias Noise=false, Noise Strength=200, Pan Noise=false)
   &Drag(
    Drag=&FloatFromCurve(FloatCurve=DataInterface() {propertyValues {Curve {
        defaultValue=1
        keys [{interpMode=RCIM_Cubic, tangentMode=RCTM_User, time=0, value=0}, {interpMode=RCIM_Cubic, tangentMode=RCTM_User, time=0.6, value=1}]
       }}})
    Use Linear Drag=true
    Use Rotational Drag=true
   )
   &SolveForcesAndVelocity(Clamp Velocity=false, Limit Acceleration=false)
  ]
  rendererClasses [&NiagaraSpriteRendererProperties]
  renderers [
   {rendererClass=&NiagaraSpriteRendererProperties}
   {rendererClass=&NiagaraSpriteRendererProperties, rendererIndex=1}
  ]
 }
 NE_VC_SingleBoost_Sprite_Parent002 {
  emitterUpdateScript [
   &EmitterState(Inactive Response=::NewEnumerator1, Life Cycle Mode=::NewEnumerator1, UseLoopDelay=false)
   &SpawnBurst_Instantaneous(Use Spawn Probability=false, UseLoopCountLimit=false)
   &SpawnRate(SpawnGroup=1, SpawnRate=3.5, Use Spawn Probability=false)
  ]
  particleSpawnScript [
   /SetVariables(Particles.VisibilityTag=$Emitter.SpawnGroup) {isSetParameters=true}
   &VC_PaintedVariantsOverride(
    Black=<rgba 0.015 0.015 0.015 6>
    BurntSienna=<rgba 0.25 0.1208 0 3>
    Cobalt=<rgba 0.0657 0.06 0.4 4>
    Crimson=<rgba 0.9 0.12 0 1.5>
    DefaultColor=<rgba 1 0.8156 0.2 1.5>
    ForestGreen=<rgba 0.2667 0.8 0 1.5>
    Grey=<rgba 0.13 0.13 0.13 4>
    Lime=<rgba 0.4667 0.8 0 2>
    Orange=<rgba 0.6 0.2117 0 1.5>
    Pink=<rgba 0.99 0.396 0.8375 1>
    Purple=<rgba 0.4104 0 1 1.5>
    R1 Black:LinearColor=<rgba 0.45 0.4114 0.38 0.7>
    R1 BurntSienna:LinearColor=<rgba 0.65 0.3142 0 1.5>
    R1 Cobalt:LinearColor=<rgba 0.5083 0.5 1 0.8>
    R1 Crimson:LinearColor=<rgba 1.5 0.25 0 1>
    R1 DefaultColor=<rgba 3 2.4469 0.6 1>
    R1 ForestGreen:LinearColor=<rgba 0.25 1 0 3>
    R1 Grey:LinearColor=<rgba 1 1 1 1.3>
    R1 Lime:LinearColor=<rgba 0.3927 0.8 0 4>
    R1 Orange:LinearColor=<rgba 1.5 0.5771 0 2>
    R1 Pink:LinearColor=<rgba 0.9 0.36 0.727 3>
    R1 Purple:LinearColor=<rgba 0.5776 0.0999 0.9131 3>
    R1 Saffron:LinearColor=<rgba 0.578 0.5493 0 4>
    R1 SkyBlue:LinearColor=<rgba 0.36 0.9307 1.2 2>
    R1 TitaniumWhite:LinearColor=<rgba 0.85 0.85 0.85 2>
    R2 Black:LinearColor=<rgba 0.8 0.7314 0.6755 0.2>
    R2 BurntSienna:LinearColor=<rgba 0.65 0.3846 0.195 1>
    R2 Cobalt:LinearColor=<rgba 0.7 0.9712 1 0.5>
    R2 Crimson:LinearColor=<rgba 1 0.5233 0.45 0.5>
    R2 DefaultColor:LinearColor=<rgba 3 2.5164 0.9016 1>
    R2 ForestGreen:LinearColor=<rgba 0.3309 1 0.2 1>
    R2 Grey:LinearColor=<rgba 5 5 5 1>
    R2 Lime:LinearColor=<rgba 1 1 0 1>
    R2 Orange:LinearColor=<rgba 1.5 1.5 0 1>
    R2 Pink:LinearColor=<rgba 1 0.922 1 1>
    R2 Purple:LinearColor=<rgba 0.7936 0.256 0.8 1>
    R2 Saffron:LinearColor=<rgba 0.8 0.7944 0.28 0.8>
    R2 SkyBlue:LinearColor=<rgba 0.7611 1.5 1.4866 1>
    Saffron=<rgba 0.8 0.6133 0 2>
    SkyBlue=<rgba 0 0.3333 1 3>
    TitaniumWhite=<rgba 1.2 1.2 1.2 1>
    Use R1 Painted Overrides=true
    Use R2 Painted Overrides=true
   )
   &InitializeParticle(
    Color=$Particles.Color
    Lifetime=&MergeHelper_FloatSwitch(0=5, 1=0.8, 2=?, 3=?, Index=$Emitter.SpawnGroup, Num=::NewEnumerator0)
    Position Offset=&AddVectorToPosition(
     Position=<-95 0 0>
     Vector=&AddVectorToPosition(Position=$User.DTCarOffset, Vector=Linked() {linkedVariable {name=User.SingleBoostPos, type=NiagaraPosition}})
    )
    Position Offset Coordinate Space=::Local
    Sprite Rotation Mode=::NewEnumerator2
    Sprite Size Mode=::NewEnumerator3
    Uniform Sprite Size=90
    UsePositionOffset=true
    Write Material Random=true
   )
   &AddVelocity(Velocity=<-1500 0 0>) {enabled=false}
  ]
  particleUpdateScript [
   &ParticleState()
   /SetVariables(Particles.CoreSpriteSize:Vector2f=&Multiply_Vector2D_ByFloat(Float=0.5, Vector2D=$Particles.Initial.SpriteSize)) {isSetParameters=true}
   &ScaleSpriteSize(Uniform Curve Sprite Scale=DataInterface() {propertyValues {Curve {
       defaultValue=0
       keys [
        {arriveTangent=4.1785, interpMode=RCIM_Cubic, leaveTangent=4.1785, tangentMode=RCTM_User, time=0, value=0}
        {arriveTangent=1.1179, interpMode=RCIM_Cubic, leaveTangent=1.1179, tangentMode=RCTM_User, time=0.2042, value=0.6228}
        {arriveTangent=0.0224, interpMode=RCIM_Cubic, leaveTangent=0.0224, tangentMode=RCTM_User, time=1, value=1}
       ]
      }}})
   &ScaleColor(
    Scale Alpha=&FloatFromCurve(FloatCurve=DataInterface() {propertyValues {Curve {
        defaultValue=1
        keys [
         {arriveTangent=9.0725, interpMode=RCIM_Cubic, leaveTangent=9.0725, tangentMode=RCTM_User, time=0, value=0}
         {arriveTangent=-2.5249, interpMode=RCIM_Cubic, leaveTangent=-2.5249, tangentMode=RCTM_User, time=0.083, value=0.9847}
         {arriveTangent=-0.0405, interpMode=RCIM_Cubic, leaveTangent=-0.0405, tangentMode=RCTM_User, time=1, value=0}
        ]
       }}})
    ScaleA=true
    ScaleRGB=true
   )
   &SolveForcesAndVelocity(Clamp Velocity=false, Limit Acceleration=false)
   &CameraOffset(Camera Offset Amount=0.1)
  ]
  rendererClasses [&NiagaraSpriteRendererProperties]
  renderers [
   {rendererClass=&NiagaraSpriteRendererProperties}
   {rendererClass=&NiagaraSpriteRendererProperties, rendererIndex=1}
   {rendererClass=&NiagaraSpriteRendererProperties, rendererIndex=2}
  ]
 }
}
summary {
 systemName=NS_Boost_Solarbeam
 emitters [
  {emitterName=NE_VC_SingleBoost_Sprite_Parent, rendererClasses [&NiagaraSpriteRendererProperties]}
  {emitterName=NE_VC_SingleBoost_Ribbon_Parent, rendererClasses [&NiagaraRibbonRendererProperties]}
  {emitterName=NE_VC_SingleBoost_Sprite_Parent002, rendererClasses [&NiagaraSpriteRendererProperties]}
 ]
 userVariables [
  User.bIsFrontendPreview {type=NiagaraBool}
  User.bIsPainted {type=NiagaraBool}
  User.ColorVariantID {type=NiagaraInt32}
  User.DTCarOffset {type=NiagaraPosition, description="", defaultValue {}}
  User.PaintedColor {type=LinearColor, description="", defaultValue {struct=LinearColor, value=<rgba 1 1 1 1>}}
  User.PaintedScalar {type=NiagaraFloat, defaultValue=1, description=""}
  User.PortCount {type=NiagaraInt32, defaultValue=1}
  User.SingleBoostPos {type=NiagaraPosition, description="", defaultValue {}}
 ]
}
